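# Editor function library for the Atom Visualizer, which gathers, analyses and batch-processes AtomModel assets.
# CreateAtomMeshesArray scans a directory recursively and pairs each model with static meshes from its first processor instance.
# Query functions take that array, blocking-load meshes, and report triangle totals, over-budget models, highest LOD count, bounds and radii.
# Processing functions run model processors on single, listed or selected models with dialog feedback; ViewportCameraFocus is a disabled stub.
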EditorUtilityBlueprint /AtomFramework/Editor/Blueprints/AtomVisualizer/Library/EFL_AtomVisualizer
graphs [
 CreateAtomMeshesArray {nodes [
   n1 = CreateAtomMeshesArray!() -> (then, DirectoryPath:String) {is=title K2Node_FunctionEntry}
   > n2 = ListAssets!(execute, DirectoryPath:String=@n1, bRecursive:bool=true, bIncludeFolder:bool=false) -> (then, ReturnValue:Array of Strings)
   > n3 = For Each Loop(Array=@n2)
   Return Node!(
    execute=@n3.Completed
    AtomMeshes:Array of STRUCT Atom Visualizer Structures=Get TempAtomMeshes() -> (TempAtomMeshes:Array of STRUCT Atom Visualizer Structures)
   )
   > n4 = LoadAsset(AssetPath=@n3)
   > n5 = Cast To AtomModel!(execute, Object=@n4) -> (then, CastFailed, AsAtom Model:Atom Model Object Reference) {is=title K2Node_DynamicCast}
   n6 = Get ModelProcessorInstances!(self:Atom Model Object Reference=@n5) -> (ModelProcessorInstances:Array of Atom Model Processor Instance Structures)
   > n7 = For Each Loop(Array:Array of Soft Object References=Break Atom Model Processor Instance!(AtomModelProcessorInstance:"Atom Model Processor Instance Structure (by ref)"="Get (a ref)"(Array:Array of Atom Model Processor Instance Structures=@n6, Dimension 1=0) -> (Output:"Atom Model Processor Instance Structure (by ref)")) -> (bUseCustomSettings:bool=~, ProcessedObjects:Array of Soft Object References) {is=title K2Node_BreakStruct}.ProcessedObjects) -> (Array Element:Soft Object Reference)
   Add(
    execute=@n7.Completed
    TargetArray:Array of STRUCT Atom Visualizer Structures=Get TempAtomMeshes()
    NewItem:"STRUCT Atom Visualizer Structure (by ref)"=Make STRUCT Atom Visualizer!(
     AtomModel_35_#1:Atom Model Soft Object Reference=Make Soft Reference!(Input:Atom Model Object Reference=@n5) -> (Output:Atom Model Soft Object Reference)
     AtomMeshes_39_#2:Array of Static Mesh Soft Object References=Get TempStaticMeshes() -> (TempStaticMeshes:Array of Static Mesh Soft Object References)
    ) -> (STRUCT_AtomVisualizer:STRUCT Atom Visualizer Structure) {is=title K2Node_MakeStruct}
   ) {is=title K2Node_CallArrayFunction}
   > Clear(TargetArray:Array of Static Mesh Soft Object References=Get TempStaticMeshes())
   n8 = LoadAsset_Blocking(execute=@n7.LoopBody, Asset=@n7)
   > n9 = Cast To StaticMesh(Object=@n8)
   > Add Unique(
    TargetArray:Array of Static Mesh Soft Object References=Get TempStaticMeshes()
    NewItem:"Static Mesh Soft Object Reference (by ref)"=Make Soft Reference(Input:Static Mesh Object Reference=@n9) -> (Output:Static Mesh Soft Object Reference)
   )
  ]}
 GetFilteredMeshes {nodes [
   n1 = GetFilteredMeshes!() -> (then, AtomMeshes:Array of STRUCT Atom Visualizer Structures, BudgetedLOD:int, TriangleBudget:int, OnlyOverBudget:bool=~) {is=title K2Node_FunctionEntry}
   > Clear(TargetArray:Array of Strings=Get StaticMeshesNames() -> (StaticMeshesNames:Array of Strings))
   > Clear(TargetArray:Array of Integers=Get StaticMeshesTriCount() -> (StaticMeshesTriCount:Array of Integers))
   > n2 = For Each Loop(Array:Array of STRUCT Atom Visualizer Structures=@n1) -> (Array Element:STRUCT Atom Visualizer Structure)
   n3 = Branch(execute=@n2.Completed, Condition=@n1.OnlyOverBudget)
   > n4 = Set FilteredDisplayIDs(FilteredDisplayIDs:Array of Integers=Get OverBudgetIDs() -> (OverBudgetIDs:Array of Integers)) -> (Output_Get:Array of Integers)
   Return Node!(
    execute=[@n4.then, @n3.else]
    FilteredMeshIDs:Array of Integers=Get FilteredDisplayIDs() -> (FilteredDisplayIDs:Array of Integers)
    OverBudgetedIDs:Array of Integers=Get OverBudgetIDs()
    HighestTriangleCounts:Array of Integers=Get StaticMeshesTriCount()
    HighestTriangleNames:Array of Strings=Get StaticMeshesNames()
   )
   Set TempTriCount(execute=@n2.LoopBody, TempTriCount:int=0) -> (Output_Get:int=0)
   n5 = Break STRUCT Atom Visualizer!(STRUCT_AtomVisualizer:"STRUCT Atom Visualizer Structure (by ref)"=@n2) -> (
    AtomModel_35_#1:Atom Model Soft Object Reference
    AtomMeshes_39_#2:Array of Static Mesh Soft Object References
   ) {is=title K2Node_BreakStruct}
   > n6 = For Each Loop(Array:Array of Static Mesh Soft Object References=@n5.AtomMeshes_39_#2) -> (Array Element:Static Mesh Soft Object Reference)
   > n7 = LoadAsset_Blocking(Asset=@n6)
   > n8 = Cast To StaticMesh(Object=@n7)
   > Set TempTriCount(TempTriCount=int + int(
     A=GetNumTriangles!(self:Static Mesh Object Reference=@n8, LODIndex:int=@n1.BudgetedLOD) -> (ReturnValue:int=0)
     B=Get TempTriCount() -> (TempTriCount:int=0)
    )) -> (Output_Get=0)
   n9 = Branch(execute=@n6.Completed, Condition="integer > integer"(A=Get TempTriCount() -> (TempTriCount=0), B=@n1.TriangleBudget))
   > Add Unique(TargetArray:Array of Integers=Get OverBudgetIDs(), NewItem:"Integer (by ref)"=@n2.Array Index)
   > n10 = LoadAsset_Blocking(Asset=@n5)
   > n11 = Cast To AtomModel(Object=@n10)
   > Add Unique(TargetArray:Array of Strings=Get StaticMeshesNames(), NewItem:"String (by ref)"=GetDisplayName(Object=@n11))
   > n12 = Add(TargetArray:Array of Integers=Get StaticMeshesTriCount(), NewItem:"Integer (by ref)"=Get TempTriCount() -> (TempTriCount=0))
   Add Unique(execute=[@n12.then, @n9.else], TargetArray:Array of Integers=Get FilteredDisplayIDs(), NewItem:"Integer (by ref)"=@n2.Array Index)
  ]}
 GetHighestLODs {nodes [
   n1 = GetHighestLODs!() -> (then, AtomMeshes:Array of STRUCT Atom Visualizer Structures, Index:int) {is=title K2Node_FunctionEntry}
   > n2 = For Each Loop(Array=Break STRUCT Atom Visualizer(STRUCT_AtomVisualizer="Get (a ref)"(Array:Array of STRUCT Atom Visualizer Structures=@n1, Dimension 1=@n1.Index) -> (Output:"STRUCT Atom Visualizer Structure (by ref)")).AtomMeshes_39_#2)
   Return Node!(
    execute=@n2.Completed
    HighestLOD:int=Get TempLOD() -> (TempLOD:int=0)
    HighestLodMesh:Static Mesh Object Reference=Get TempStaticMesh() -> (TempStaticMesh:Static Mesh Object Reference)
   )
   > n3 = LoadAsset_Blocking(Asset=@n2)
   > n4 = Cast To StaticMesh(Object=@n3)
   n5 = GetNumLODs(self=@n4)
   > Branch(Condition="integer > integer"(A=@n5, B=Get TempLOD() -> (TempLOD=0)))
   > Set TempLOD(TempLOD:int=@n5) -> (Output_Get:int=0)
   > Set TempStaticMesh(TempStaticMesh:Static Mesh Object Reference=@n4) -> (Output_Get:Static Mesh Object Reference)
  ]}
 GetLargestBound {nodes [
   n1 = GetLargestBound!() -> (then, AtomMeshes:Array of STRUCT Atom Visualizer Structures) {is=title K2Node_FunctionEntry}
   > n2 = For Each Loop(Array:Array of STRUCT Atom Visualizer Structures=@n1) -> (Array Element:STRUCT Atom Visualizer Structure)
   Return Node!(
    execute=@n2.Completed
    LargestBound:double=Get TempBaseSize() -> (TempBaseSize:double)
    LargestBoundMesh:Static Mesh Object Reference=Get TempStaticMesh()
   )
   > n3 = For Each Loop(Array:Array of Static Mesh Soft Object References=Break STRUCT Atom Visualizer(STRUCT_AtomVisualizer=@n2).AtomMeshes_39_#2) -> (Array Element:Static Mesh Soft Object Reference)
   > n4 = LoadAsset_Blocking(Asset=@n3)
   > n5 = Cast To StaticMesh(Object=@n4)
   n6 = Break Box!(Box:"Box Structure (by ref)"=GetBoundingBox!(self:Static Mesh Object Reference=@n5) -> (ReturnValue:Box Structure)) -> (Min_X:double, Min_Y:double, Min_Z:double, Max_X:double, Max_Y:double, Max_Z:double, IsValid:bool=~) {is=title K2Node_BreakStruct}
   n7 = float - float(A=@n6.Max_Y, B=@n6.Min_Y)
   n8 = float - float(A=@n6.Max_X, B=@n6)
   n9 = Sqrt!(A:double=float + float(A=float * float(A=@n8, B=@n8) {is=title K2Node_PromotableOperator}, B=float * float(A=@n7, B=@n7))) -> (ReturnValue:double)
   > Branch(Condition="float > float"(A=@n9, B=Get TempBaseSize()))
   > Set TempBaseSize(TempBaseSize:double=@n9) -> (Output_Get:double)
   > Set TempStaticMesh(TempStaticMesh=@n5)
  ]}
 GetNumValidStaticMeshes {nodes [
   n1 = GetNumValidStaticMeshes!() -> (then, AtomMeshes:Array of STRUCT Atom Visualizer Structures) {is=title K2Node_FunctionEntry}
   > n2 = For Each Loop(Array:Array of STRUCT Atom Visualizer Structures=@n1) -> (Array Element:STRUCT Atom Visualizer Structure)
   Return Node!(execute=@n2.Completed, number of Static Meshes:int=Get Counter() -> (Counter:int=0))
   > Branch(Condition=Is Not Empty!(TargetArray:Array of Static Mesh Soft Object References=Break STRUCT Atom Visualizer(STRUCT_AtomVisualizer=@n2).AtomMeshes_39_#2) -> (ReturnValue:bool))
   > Increment Int(Value=Get Counter() -> (Counter=0))
  ]}
 GetRadiusInfo {nodes [
   n1 = GetRadiusInfo!() -> (then, AtomMeshes:Array of STRUCT Atom Visualizer Structures) {is=title K2Node_FunctionEntry}
   > n2 = For Each Loop(Array=@n1)
   > n3 = For Each Loop(Array:Array of Static Mesh Soft Object References=Break STRUCT Atom Visualizer(STRUCT_AtomVisualizer=@n2).AtomMeshes_39_#2) -> (Array Element:Static Mesh Soft Object Reference)
   > n4 = LoadAsset_Blocking(Asset=@n3)
   > n5 = Cast To StaticMesh(Object=@n4)
   > n6 = Branch(Condition="Equal (Integer)"(A=@n2.Array Index))
   n7 = BreakBoxSphereBounds!(InBoxSphereBounds:"Box Sphere Bounds Structure (by ref)"=GetBounds!(self:Static Mesh Object Reference=@n5) -> (ReturnValue:Box Sphere Bounds Structure)) -> (Origin:Vector, BoxExtent:Vector, SphereRadius:float)
   > n8 = Set TempSmallest(TempSmallest:double=@n7.SphereRadius) -> (Output_Get:double)
   n9 = Branch(execute=[@n8.then, @n6.else], Condition="float < float"(A=@n7.SphereRadius, B=Get TempSmallest() -> (TempSmallest:double)))
   > n10 = Set TempSmallest(TempSmallest=@n7.SphereRadius)
   n11 = Branch(execute=[@n10.then, @n9.else], Condition="float > float"(A=@n7.SphereRadius, B=Get TempLargest() -> (TempLargest:double)))
   > n12 = Set TempLargest(TempLargest:double=@n7.SphereRadius) -> (Output_Get:double)
   Set SumRadiuses(execute=[@n12.then, @n11.else], SumRadiuses:double=float + float(A=@n7.SphereRadius, B=Get SumRadiuses() -> (SumRadiuses:double))) -> (Output_Get:double)
   n13 = Length(TargetArray:Array of STRUCT Atom Visualizer Structures=@n1)
   n14 = Get SumRadiuses()
   Return Node!(
    execute=@n2.Completed
    AverageRadius:int=Round(A=float / float(A=@n14, B:int=@n13)) {is=title K2Node_CallFunction}
    LargestRadius:int=Round(A=Get TempLargest())
    SmallestRadius:int=Round(A=Get TempSmallest())
   )
  ]}
 GetTriangleCountAtomMesh {nodes [
   n1 = GetTriangleCountAtomMesh!() -> (then, AtomMeshes:Array of STRUCT Atom Visualizer Structures, LODIndex:int) {is=title K2Node_FunctionEntry}
   > n2 = For Each Loop(Array:Array of STRUCT Atom Visualizer Structures=@n1) -> (Array Element:STRUCT Atom Visualizer Structure)
   n3 = Get MeshCounter() -> (MeshCounter:int=0)
   n4 = Get TotalTriangles() -> (TotalTriangles:int=0)
   Return Node!(execute=@n2.Completed, TotalTriangleCount:int=@n4, AverageTriangleCount:int=int / int(A=@n4, B=@n3), NumModels:int=@n3)
   > n5 = For Each Loop(Array:Array of Static Mesh Soft Object References=Break STRUCT Atom Visualizer(STRUCT_AtomVisualizer=@n2).AtomMeshes_39_#2) -> (Array Element:Static Mesh Soft Object Reference)
   Increment Int(" "=@n5.Completed, Value=Get MeshCounter() -> (MeshCounter=0))
   n6 = LoadAsset_Blocking(execute=@n5.LoopBody, Asset=@n5)
   > n7 = Cast To StaticMesh(Object=@n6)
   n8 = "Max (Integer)"(A=int - int(A=@n1.LODIndex), B=0)
   > Set TotalTriangles(TotalTriangles:int=int + int(A=Get TotalTriangles() -> (TotalTriangles=0), B=GetNumTriangles(self=@n7, LODIndex=@n8) -> (ReturnValue=0))) -> (Output_Get:int=0)
  ]}
 ProcessAtomModel {nodes [
   n1 = ProcessAtomModel!() -> (then, Atom Model:Atom Model Soft Object Reference, ShowDialog:bool=~) {is=title K2Node_FunctionEntry}
   > n2 = LoadAsset_Blocking(Asset=@n1)
   > n3 = Cast To AtomModel(Object=@n2)
   > RunModelProcessors!(execute, AtomModel:Atom Model Object Reference=@n3) -> (then, ReturnValue:Array of Atom Processor Result Structures)
   > n4 = Branch(Condition=@n1.ShowDialog)
   n5 = Append!(A="Done! \r\n\r\n", B=GetDisplayName(Object=@n3), C:String=" has been Processed.\r\n\r\n") -> (ReturnValue)
   > n6 = Show Message Dialog(
    Title=MakeLiteralText(Value="LOCTEXT(\"Process Selected Atom Model\")")
    Message=MakeLiteralText(Value="To Text (String)"(InString=@n5))
    MessageType=Ok
    DefaultValue=Ok
    MessageCategory=Success
   )
   n7 = Set Success(execute=[@n6.then, @n4.else], Success:bool=true) -> (Output_Get:bool)
   Return Node!(execute=[@n7.then, @n8.then], Success?:bool=Get Success() -> (Success:bool))
   n8 = Set Success(execute=@n3.CastFailed, Success=false)
  ]}
 ProcessAtomModels {nodes [
   n1 = ProcessAtomModels!() -> (then, ShowDialog:bool=~, AtomMeshes:Array of STRUCT Atom Visualizer Structures) {is=title K2Node_FunctionEntry}
   > n2 = For Each Loop(Array:Array of STRUCT Atom Visualizer Structures=@n1.AtomMeshes) -> (Array Element:STRUCT Atom Visualizer Structure)
   n3 = Branch(execute=@n2.Completed, Condition=@n1)
   n4 = Append(A="Done! \r\n\r\n", B="To String (Integer)"(InInt=Get NumModels() -> (NumModels:int=0)), C=" Atom Models Processed.")
   > n5 = Show Message Dialog(
    Title=MakeLiteralText(Value="LOCTEXT(\"Batch Process Atom Assets\")")
    Message=MakeLiteralText(Value="To Text (String)"(InString=@n4))
    MessageType=Ok
    DefaultValue=Ok
    MessageCategory=Success
   )
   Return Node!(execute=[@n5.then, @n3.else], NumAssets:int=Get NumModels() -> (NumModels=0))
   n6 = LoadAsset_Blocking(execute=@n2.LoopBody, Asset=Break STRUCT Atom Visualizer(STRUCT_AtomVisualizer=@n2))
   > n7 = Cast To AtomModel(Object=@n6)
   > n8 = ProcessAtomModel!(
    execute
    Atom Model:Atom Model Soft Object Reference=Make Soft Reference(Input:Atom Model Object Reference=@n7) -> (Output:Atom Model Soft Object Reference)
    ShowDialog:bool=false
    __WorldContext:Object Reference
   ) -> (then, Success?:bool) {is=title K2Node_CallFunction}
   > Branch(Condition=@n8)
   > Increment Int(Value=Get NumModels() -> (NumModels=0))
  ]}
 ProcessSelectedAtomModels {nodes [
   ProcessSelectedAtomModels() {is=title K2Node_FunctionEntry}
   > n1 = GetSelectedAssetsOfClass(AssetClass=/Script/AtomRuntime.AtomModel)
   > n2 = For Each Loop(Array:Array of Object References=@n1) -> (Array Element:Object Reference)
   n3 = Append(A="Done! \r\n", B="To String (Integer)"(InInt=Get NumModels() -> (NumModels=0)), C=" Atom Models Processed!")
   Show Message Dialog(
    execute=@n2.Completed
    Title=MakeLiteralText(Value="LOCTEXT(\"Process Selected Atom Model\")")
    Message=MakeLiteralText(Value="To Text (String)"(InString=@n3))
    MessageType=Ok
    DefaultValue=Ok
    MessageCategory=Success
   )
   > Return Node!(execute)
   n4 = Cast To AtomModel(execute=@n2.LoopBody, Object=@n2)
   > n5 = ProcessAtomModel(Atom Model=Make Soft Reference(Input=@n4), ShowDialog=false)
   > Branch(Condition=@n5)
   > Increment Int(Value=Get NumModels() -> (NumModels=0))
  ]}
 ViewportCameraFocus {nodes [
   ViewportCameraFocus!() -> (then, Actors:Array of Actor Object References) {is=title K2Node_FunctionEntry}
   > PrintString(InString=Function temporarily disabled due to problems on GetEditorCameraInfo)
   > Return Node()
  ]}
]
variables []
